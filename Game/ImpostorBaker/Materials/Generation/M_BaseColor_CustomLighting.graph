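# Unlit alpha-composite ImpostorBaker generation material that combines a BaseColor capture with a custom lighting texture for impostor baking.
# The lighting sample is saturated, desaturated by a parameter, and raised to a clamped power from MPC_ImposterCapture.
# A collection opacity value blends that result from white before it multiplies the base color.
# The inverted BaseColor alpha forms a second output.
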Material /Game/ImpostorBaker/Materials/Generation/M_BaseColor_CustomLighting
graph {
 properties {blendMode=BLEND_AlphaComposite, pixelDepthOffsetMode=PDOM_Legacy, shadingModel=MSM_Unlit, usedShadingModels=Unlit}
 nodes [
  n1 = Desaturation(
   None=Saturate(None=TextureSampleParameter2D() {parameterName=CustomLighting, texture=&Black}.RGB)
   Fraction=ScalarParameter() {defaultValue=0.5, parameterName=Desat}
  )
  n2 = Max(A=CollectionParameter() {collection=/Game/ImpostorBaker/MPC/MPC_ImposterCapture, parameterName=CustomLightingPower}) {constB=0.001}
  n3 = CollectionParameter() {collection=&MPC_ImposterCapture, parameterName=CustomLightingBasecolorOpacity}
  n4 = TextureSampleParameter2D() {parameterName=BaseColor, texture=&Black}
  Multiply(A=LinearInterpolate(A=Constant() {r=1}, B=Power(Base=@n1, Exp=@n2), Alpha=Saturate(None=@n3)) {constA=1}, B=@n4.RGB)
  OneMinus(None=@n4.A)
 ]
}
