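# Unlit alpha-composite material in the ImpostorBaker generation set that, going by its name, packs normals and depth together while impostor textures are baked.
# It samples Normal and Depth texture parameters, both defaulting to black, and inverts the Depth red channel with a OneMinus.
# The output pins are not shown; the setup suggests normals fill color and inverted depth fills alpha.
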Material /Game/ImpostorBaker/Materials/Generation/M_CombineNormalsDepth
graph {
 properties {blendMode=BLEND_AlphaComposite, pixelDepthOffsetMode=PDOM_Legacy, shadingModel=MSM_Unlit, usedShadingModels=Unlit}
 nodes [
  OneMinus(None=TextureSampleParameter2D() {parameterName=Depth, texture=&Black}.R)
  TextureSampleParameter2D() {parameterName=Normal, texture=&Black}
 ]
}
