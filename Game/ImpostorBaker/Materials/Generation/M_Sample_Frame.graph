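# Unlit alpha-composite material for impostor baking that samples the RT render target and dilates captured colours into empty texels.
# A Custom HLSL node steps in eight directions from snapped texel-centre UVs to find the nearest low-alpha texel, with an If supplying a grey fallback.
# One lerp blends an unsharp-masked RT with that result by alpha.
# Another blends raw RT with it using the Dilation value from MPC_ImposterCapture.
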Material /Game/ImpostorBaker/Materials/Generation/M_Sample_Frame
graph {
 properties {blendMode=BLEND_AlphaComposite, pixelDepthOffsetMode=PDOM_Legacy, shadingModel=MSM_Unlit, usedShadingModels=Unlit}
 nodes [
  MaterialExpressionTextureObjectParameter_0 = TextureObjectParameter() {parameterName=RT, texture=&Black}
  MaterialExpressionScalarParameter_12 = ScalarParameter() {defaultValue=512, parameterName=Res}
  n1 = Multiply(A=TextureCoordinate(), B=@MaterialExpressionScalarParameter_12) {constB=1024}
  MaterialExpressionAdd_0 = Add(
   A=Divide(A=Floor(None=@n1), B=@MaterialExpressionScalarParameter_12) {constB=1024}
   B=Divide(B=@MaterialExpressionScalarParameter_12) {constA=0.5, constB=2048}
  )
  MaterialExpressionScalarParameter_13 = ScalarParameter() {defaultValue=64, parameterName=ms}
  n2 = Custom!(
   Tex=@MaterialExpressionTextureObjectParameter_0
   UV=@MaterialExpressionAdd_0
   Resolution=@MaterialExpressionScalarParameter_12
   MaxSteps=@MaterialExpressionScalarParameter_13
   TextureSize=@MaterialExpressionScalarParameter_12
  ) {
   code="//////////////// UV Dilation ///////////////////////////\r\n\r\n\r\n//** Tex        **// Input Texture Object storing Volume Data\r\n//** UV         **// Input float2 for UVs\r\n//** Resolution **// Resolution of render target\r\n\r\n\r\nfloat texelsize = 1 / Resolution;\r\nfloat mindist = 10000000;\r\n\r\nfloat2 offsets[8] = {float2(-1,0), float2(1,0), float2(0,1), float2(0,-1), float2(-1,1), float2(1,1), float2(1,-1), float2(-1,-1)};\r\n\r\nfloat4 sample = Tex.SampleLevel(TexSampler, UV, 0);\r\n\r\nfloat4 curminsample = sample;\r\n\r\n\r\n\r\nif(sample.a > 0)\r\n{\r\n\r\n\tint j = 0;\r\n\twhile(j < MaxSteps)\r\n\t{\r\n\t\t\r\n\t\tj++;\r\n\r\n\t\tint i = 0;\r\n\r\n\t\twhile (i < 8)\r\n\t\t{\r\n\r\n\t\t\tfloat2 curUV = UV + offsets[i] * texelsize * j;\r\n\t\t\tfloat4 offsetsample = Tex.SampleLevel(TexSampler, curUV, 0);\r\n\r\n\t\t\tif(offsetsample.a <= 0.01)\r\n\t\t\t{\r\n\r\n\t\t\t\tfloat curdist = length(UV - curUV);\r\n\r\n\t\t\t\tif (curdist < mindist)\r\n\t\t\t\t{\r\n\t\t\t\t\tfloat2 projectUV = curUV + offsets[i] * texelsize * j * 1;\r\n\t\t\t\t\tfloat3 direction = Tex.SampleLevel(TexSampler, projectUV, 0);\r\n\t\t\t\t\t\r\n\t\t\t\t\tif(direction.x != 0 || direction.y != 0 || direction.z != 0)\r\n\t\t\t\t\t{\r\n\t\t\t\t\t\r\n\t\t\t\t\t\tmindist = curdist;\r\n\t\t\t\t\t\tcurminsample = offsetsample;\r\n\r\n\t\t\t\t\t}\r\n\t\t\t\t\telse\r\n\t\t\t\t\t{\r\n\t\t\t\t\tmindist = curdist;\r\n\t\t\t\t\tcurminsample = offsetsample;\r\n\t\t\t\t\t}\r\n\t\t\t\t}\r\n\t\t\t}\r\n\r\n\t\r\n\t\t\ti++;\r\n\t\t}\r\n\r\n\t}\r\n}\r\n\t\r\nreturn curminsample.xyz;"
   inputs [
    {inputName=Tex, input {expression=&self:MaterialExpressionTextureObjectParameter_0}}
    {inputName=UV, input {expression=&self:MaterialExpressionAdd_0}}
    {inputName=Resolution, input {expression=&self:MaterialExpressionScalarParameter_12}}
    {inputName=MaxSteps, input {expression=&self:MaterialExpressionScalarParameter_13}}
    {inputName=TextureSize, input {expression=&self:MaterialExpressionScalarParameter_12}}
   ]
  }
  n3 = Constant() {r=0.5}
  n4 = If(A=ComponentMask(None=@n2) {r=true}, "A > B"=@n2, "A == B"=@n3, "A < B"=@n3)
  n5 = TextureSampleParameter2D() {parameterName=RT, texture=&Black}
  LinearInterpolate(
   A=MaterialFunctionCall!(
    Texture Object=@MaterialExpressionTextureObjectParameter_0
    UVs=TextureCoordinate()
    Amount=ScalarParameter() {defaultValue=0.5, parameterName=Amt}
    Sample Offset=Constant() {r=2}
    High Quality
    TextureSize
   ) {materialFunction=/Engine/Functions/Engine_MaterialFunctions01/ImageAdjustment/UnSharpMaskTexture}.Result
   B=@n4
   Alpha=@n5.A
  )
  LinearInterpolate(A=@n5.RGB, B=@n4, Alpha=CollectionParameter() {collection=/Game/ImpostorBaker/MPC/MPC_ImposterCapture, parameterName=Dilation})
  OneMinus(None=@n5.A)
  TextureObjectParameter() {parameterName=MipRT, texture=&Black}
  TextureObject() {texture=&Black}
  TextureSampleParameter2D() {parameterName=MipRT, texture=&Black}
 ]
}
